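# Parent material for the WolfHunter medium male character's face accessory, adding a parallax emissive layer, apparently for the eyes, to the shared character shading.
# It sits between the 2023 character PreFX and PostFX functions, overriding Emissive Color on the attributes passed between them.
# A bump-offset emissive texture is scaled by brightness, with a multiplier outside deferred shading for mobile.
# Inverted Fresnel and the Tech Art Mask's red channel limit the glow.
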Material /BRCosmetics/Characters/Player/Male/Medium/Bodies/M_MED_WolfHunter/Materials/M_WolfHunter_FaceAcc_Parent
graph {
 properties {
  bUseMaterialAttributes=true
  bUsedWithMeshDeformer=true
  bUsedWithMorphTargets=true
  bUsedWithSkeletalMesh=true
  pixelDepthOffsetMode=PDOM_Legacy
  shadingModel=MSM_FromMaterialExpression
  usedShadingModels=Default Lit | Subsurface | Clear Coat | Subsurface Profile | Cloth
 }
 nodes [
  n1 = MaterialFunctionCall() {materialFunction=&MF_CharacterParent_2023_PreFX}
  n2 = ComponentMask(None=@n1.Tech Art Mask) {r=true}
  n3 = TextureSampleParameter2D(UVs=MaterialFunctionCall!(
    UV Coord=TextureCoordinate()
    Height=ScalarParameter() {defaultValue=0.1, parameterName=EyesHeight}
    ReferencePlane=Constant()
    HeightRatio
    Normal=NamedRerouteUsage()
   ) {materialFunction=/Game/Global/MaterialFunctions/MF_BumpOffset_Accurate}.Result) {
   group=06_Emissive
   parameterName=Emissive
   samplerSource=SSM_Clamp_WorldGroupSettings
   sortPriority=1
   texture=&T_BlacksRGB
  }
  n4 = ScalarParameter() {defaultValue=1, group=06_Emissive, parameterName=EmissiveBrightness}
  n5 = ShadingPathSwitch(Default=ScalarParameter() {defaultValue=1, group=06_Emissive, parameterName=EmissiveMobileMultiplier}, Deferred=Constant() {r=1})
  n6 = Multiply(A=OneMinus(None=@n1.Fresnel), B=Multiply(A=@n3.RGB, B=Multiply(A=@n4, B=@n5)))
  MaterialFunctionCall(
   Material=SetMaterialAttributes(MaterialAttributes=@n1.Result, Emissive Color=Multiply(A=@n2, B=@n6)) {attributeSetTypes [#1]}
   M=@n1.M
   Pre Skinned Local Position=@n1.M
   Distance Blend=@n1.M
   Fresnel=@n1.M
  ) {materialFunction=&MF_CharacterParent_2023_PostFX}
  NamedRerouteDeclaration(None=OneMinus(None=Fresnel() {baseReflectFraction=0, exponent=1})) {name=Fresnel, nodeColor=<rgba 1 0.8203 0 1>}
  NamedRerouteDeclaration(None=GetMaterialAttributes(None=@n1.Result) {attributeGetTypes [#2]}.Normal) {name=Normal_IN, nodeColor=<rgba 1 0 0.7266 1>}
 ]
}
