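# Animation Blueprint for the WolfHunter medium male face accessory mesh, following the character it is attached to.
# It copies the pose from the attached parent, runs it through the accessory's Control Rig and caches the result.
# A multi-target Constraint node then drives base_hood_c from weighted head, neck, upper spine and clavicle bones, weighted most toward the head.
# The event graph holds no logic and there are no variables.
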AnimBlueprint /BRCosmetics/Characters/Player/Male/Medium/Bodies/M_MED_WolfHunter/Meshes/Parts/M_MED_WolfHunter_FaceAcc_AnimBP
target_skeleton=&/M_MED_WolfHunter_FaceAcc_Skeleton
graphs [
 AnimGraph {nodes [
   n1 = "Constraint - base_hood_c (PPPPPPPP)"!(
    ComponentPose=Local To Component(LocalPose=Use cached pose 'CR Cache Input'())
    Alpha=1
    ConstraintWeights_0=0.4742
    ConstraintWeights_1:float=0.2779
    ConstraintWeights_2:float=0.1542
    ConstraintWeights_3:float=0.0569
    ConstraintWeights_4:float=0.0175
    ConstraintWeights_5:float=0.0129
    ConstraintWeights_6:float=0.0062
    ConstraintWeights_7:float=0.0002
   ) -> (Pose) {
    alphaBoolBlend {blendOption=Linear}
    boneToModify {boneName=base_hood_c}
    constraintSetup [
     {targetBone {boneName=head}}
     {targetBone {boneName=spine_05}}
     {targetBone {boneName=neck_02}}
     {targetBone {boneName=neck_01}}
     {targetBone {boneName=clavicle_l}}
     {targetBone {boneName=spine_03}}
     {targetBone {boneName=clavicle_r}}
     {targetBone {boneName=spine_04}}
    ]
    constraintWeights [1,1,1,1,1,1,1,1]
   }
   Output Pose(Result=Component To Local(ComponentPose=@n1)) {layerGroup=None}
   n2 = Control Rig(Source=Copy Pose From Mesh() {bUseAttachedParent=true}) {
    alphaBoolBlend {blendOption=Linear}
    controlRigAssetReference {blueprintRigClass=&/M_MED_WolfHunter_FaceAcc_CtrlRig.M_MED_WolfHunter_FaceAcc_CtrlRig_C}
   }
   Save cached pose 'CR Cache Input'(Pose=@n2) {cacheName=CR Cache Input}
  ]}
 EventGraph {nodes [Event BlueprintUpdateAnimation()]}
]
variables []
